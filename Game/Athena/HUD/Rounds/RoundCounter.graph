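# RoundCounter is a minigame HUD widget showing "ROUND current/total" and the owning player's win count beside a trophy icon.
# Warmup and in-progress display events set flags, then run a shared sequence that first sets visibility, showing only when either flag is true outside curated hubs.
# While either flag is set, it fills the round text from the current minigame and the wins from the player's scoreboard row, updating only on those events.
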WidgetBlueprint /Game/Athena/HUD/Rounds/RoundCounter
graphs [EventGraph {nodes [
   n1 = Set Round Counter!() -> (OutputDelegate, then, Current:int, Total:int) {is=title K2Node_CustomEvent}
   > "SetText (Text)"(
    self=Get TextBlock_Round() -> (TextBlock_Round:Common Text Object Reference)
    InText=Format Text!(
     Format="LOCTEXT(\"ROUND {Current}/{Total}\")"
     Current:Text="To Text (Integer)"(Value=@n1.Current, MinimumIntegralDigits=2)
     Total:Text="To Text (Integer)"(Value=@n1.Total, MinimumIntegralDigits=2)
    ) -> (Result)
   )
   n2 = Set Win Counter!() -> (OutputDelegate, then, Value:int) {is=title K2Node_CustomEvent}
   > "SetText (Text)"(self=Get TextBlock_Wins() -> (TextBlock_Wins:Common Text Object Reference), InText="To Text (Integer)"(Value=@n2.Value))
   n3 = Event OnDisplayMinigameInProgressUI!() -> (OutputDelegate, then, bDisplay:bool) {is=title K2Node_Event}
   > n4 = Set bIsInProgress(bIsInProgress:bool=@n3.bDisplay) -> (Output_Get:bool)
   n5 = Sequence(execute=[@n16.then, @n4.then])
   n6 = OR Boolean(A=Get bIsWarmup() -> (bIsWarmup:bool), B=Get bIsInProgress() -> (bIsInProgress:bool))
   Branch(execute=@n5.then_1, Condition=@n6)
   > n7 = Cast To FortPlayerState!(execute, Object=Get PlayerState!(self:Controller Object Reference=GetOwningPlayer()) -> (PlayerState:Player State Object Reference)) -> (then, CastFailed, AsFort Player State:Fort Player State Object Reference) {is=title K2Node_DynamicCast}
   > n8 = Get!(execute:Exec) -> (then:Exec, else:Exec, CurrentMinigame:Fort Minigame Object Reference)
   > Set Round Counter!(
    execute
    self
    Current:int=Get CurrentRound!(self:Fort Minigame Object Reference=@n8) -> (CurrentRound:int=0)
    Total:int=Get TotalRounds!(self:Fort Minigame Object Reference=@n8) -> (TotalRounds:int=0)
   ) -> (then) {is=title K2Node_CallFunction}
   > n9 = Get!(execute, self:Fort Minigame Object Reference=@n8) -> (then, else, ScoreboardComponent:Fort Minigame Scoreboard Component Object Reference)
   n10 = Get ScoreboardStatObjects!(self:Fort Minigame Stats Component Object Reference=GetMinigameStatsComponent!(ContextActor:Actor Object Reference=Get Volume!(self:Fort Minigame Object Reference=@n8) -> (Volume:Actor Object Reference)) -> (ReturnValue:Fort Minigame Stats Component Object Reference)) -> (ScoreboardStatObjects:Array of Fort Minigame Stat Filter Object References)
   n11 = Get ScoreboardRowParams() -> (ScoreboardRowParams:Minigame Scoreboard Row Params Structure)
   > Set Win Counter!(
    execute
    self
    Value:int=Break Minigame Scoreboard Player Row!(MinigameScoreboardPlayerRow:"Minigame Scoreboard Player Row Structure (by ref)"=GetScoreboardRowForPlayer!(
      self:Fort Minigame Scoreboard Component Object Reference=@n9
      PlayerState:Fort Player State Object Reference=@n7
      SortOrder:Array of Fort Minigame Stat Filter Object References=@n10
      InParams:"Minigame Scoreboard Row Params Structure (by ref)"=@n11
     ) -> (OutScoreboardRow:Minigame Scoreboard Player Row Structure, ReturnValue:bool)) -> (
     PlayerState:Fort Player State Object Reference
     bHasLeftGame:bool=~
     bHideFromDisplay:bool=~
     Name:Text
     TeamColorIndex:int
     Values:Array of Minigame Scoreboard Value Structures
     WinCount:int
     bHighlight:bool=~
    ) {is=title K2Node_BreakStruct}.WinCount
   ) -> (then) {is=title K2Node_CallFunction}
   n12 = Get!(execute=@n5.then_0) -> (then, else, CurrentMinigame:Fort Minigame Object Reference)
   n13 = NOT Boolean(A=IsCuratedHub!(self:Fort Minigame Object Reference=@n12) -> (ReturnValue:bool))
   n14 = OR Boolean(A=Get bIsWarmup(), B=Get bIsInProgress())
   > SetVisibility(InVisibility=Select(Option 0=Collapsed, Option 1=SelfHitTestInvisible, Index=AND Boolean(A=@n13, B=@n14)))
   SetVisibility(execute=@n12.else, InVisibility=Collapsed)
   n15 = Event OnDisplayMinigameStartUI!() -> (OutputDelegate, then, bDisplay:bool) {is=title K2Node_Event}
   > n16 = Set bIsWarmup(bIsWarmup:bool=@n15.bDisplay) -> (Output_Get:bool)
   AND Boolean(A=false, B=@n6)
  ]}]
variables [
 bIsInProgress {type=boolean}
 bIsWarmup {type=boolean}
 ScoreboardRowParams {
  type=MinigameScoreboardRowParams
  default {
   bCountWins=true
   bGameTotal=false
   bHighlighting=false
   bOverrideIncludeInactivePlayers=false
   rankingMode=StandardCompetition
   scoreboardState=Max
  }
 }
]
widgets {
 info {inheritedWidgetCount=1, parentClass=/Script/FortniteUI.MinigameWidgetBase, rootWidgetClass=&SizeBox, widgetCount=10}
 widgets [
  SizeBox_0 {widgetClassPath=&SizeBox, widget=SizeBox_0, properties {bOverride_WidthOverride=true, widthOverride=300}}
  RoundsContainer {widgetClassPath=&HorizontalBox, bIsVariable=true, parent=SizeBox_0, widget=RoundsContainer, slot {class=SizeBoxSlot}}
  HorizontalBox {
   widgetClassPath=&HorizontalBox
   parent=RoundsContainer
   widget=HorizontalBox
   slot {class=HorizontalBoxSlot, properties {horizontalAlignment=HAlign_Left, verticalAlignment=VAlign_Center}}
  }
  TextBlock_Round {
   widgetClassPath=&CommonTextBlock
   bIsVariable=true
   parent=HorizontalBox
   widget=TextBlock_Round
   properties {
    text=ROUND 5/25
    font {
     bForceMonospaced=false
     bMaterialIsStencil=false
     fontMaterial=null
     fontObject=&BurbankSmall
     letterSpacing=0
     monospacedWidth=1
     size=16
     skewAmount=0
     typefaceFontName=Black
     outlineSettings {
      bApplyOutlineToDropShadows=false
      bMiteredCorners=false
      bSeparateFillAlpha=false
      outlineMaterial=null
      outlineSize=1
      outlineColor=<rgba 0 0 0 0.4>
     }
    }
    shadowOffset=<0 0>
   }
   slot {class=HorizontalBoxSlot, properties {padding=<ltrb 10 0 0 0>}}
  }
  Overlay_0 {
   widgetClassPath=&Overlay
   parent=RoundsContainer
   widget=Overlay_0
   slot {
    class=HorizontalBoxSlot
    properties {horizontalAlignment=HAlign_Right, verticalAlignment=VAlign_Center, padding=<ltrb 0 0 4 0>, size {sizeRule=Fill, value=8}}
   }
  }
  TrophyIcon_HackOutline {
   widgetClassPath=&Image
   parent=Overlay_0
   widget=TrophyIcon_HackOutline
   properties {
    visibility=SelfHitTestInvisible
    brush {
     bIsDynamicallyLoaded=false
     drawAs=Image
     imageType=NoImage
     mirroring=NoMirror
     resourceName=None
     resourceObject=/Game/UI/Foundation/Textures/Icons/Quest/T-Icon-Trophy-32
     tiling=NoTile
     imageSize=<27 26.5>
     margin=<ltrb 0 0 0 0>
     outlineSettings {
      bUseBrushTransparency=false
      roundingType=HalfHeightRadius
      width=0
      color {colorUseRule=UseColor_Specified, specifiedColor=<rgba 0 0 0 0>}
      cornerRadii=<0 0 0 0>
     }
     tintColor {colorUseRule=UseColor_Specified, specifiedColor=<rgba 0 0 0 0.4>}
     uVRegion {bIsValid=false, max=<0 0>, min=<0 0>}
    }
   }
   slot {class=OverlaySlot, properties {horizontalAlignment=HAlign_Center, verticalAlignment=VAlign_Center, padding=<ltrb -1.4 -1.5 1 -1>}}
  }
  TrophyIcon_HackOutline_1 {
   widgetClassPath=&Image
   parent=Overlay_0
   widget=TrophyIcon_HackOutline_1
   properties {
    visibility=SelfHitTestInvisible
    brush {
     bIsDynamicallyLoaded=false
     drawAs=Image
     imageType=NoImage
     mirroring=NoMirror
     resourceName=None
     resourceObject=/Game/Athena/Collections/Art/Character/M_Ui_RadialGradient_Bottom
     tiling=NoTile
     imageSize=<17.5159 15.1055>
     margin=<ltrb 0 0 0 0>
     outlineSettings {
      bUseBrushTransparency=false
      roundingType=HalfHeightRadius
      width=0
      color {colorUseRule=UseColor_Specified, specifiedColor=<rgba 0 0 0 0>}
      cornerRadii=<0 0 0 0>
     }
     tintColor {colorUseRule=UseColor_Specified, specifiedColor=<rgba 0 0 0 0.2>}
     uVRegion {bIsValid=false, max=<0 0>, min=<0 0>}
    }
   }
   slot {class=OverlaySlot, properties {horizontalAlignment=HAlign_Center, verticalAlignment=VAlign_Center, padding=<ltrb 0 2 2.5 1>}}
  }
  TrophyIcon {
   widgetClassPath=&Image
   parent=Overlay_0
   widget=TrophyIcon
   properties {
    visibility=SelfHitTestInvisible
    brush {
     bIsDynamicallyLoaded=false
     drawAs=Image
     imageType=NoImage
     mirroring=NoMirror
     resourceName=None
     resourceObject=&T-Icon-Trophy-32
     tiling=NoTile
     imageSize=<22 22>
     margin=<ltrb 0 0 0 0>
     outlineSettings {
      bUseBrushTransparency=false
      roundingType=HalfHeightRadius
      width=0
      color {colorUseRule=UseColor_Specified, specifiedColor=<rgba 0 0 0 0>}
      cornerRadii=<0 0 0 0>
     }
     tintColor {colorUseRule=UseColor_Specified, specifiedColor=<rgba 1 1 1 1>}
     uVRegion {bIsValid=false, max=<0 0>, min=<0 0>}
    }
   }
   slot {class=OverlaySlot}
  }
  TextBlock_Wins {
   widgetClassPath=&CommonTextBlock
   bIsVariable=true
   parent=RoundsContainer
   widget=TextBlock_Wins
   properties {
    text=2
    font {
     bForceMonospaced=false
     bMaterialIsStencil=false
     fontMaterial=null
     fontObject=&BurbankSmall
     letterSpacing=0
     monospacedWidth=1
     size=16
     skewAmount=0
     typefaceFontName=Black
     outlineSettings {
      bApplyOutlineToDropShadows=false
      bMiteredCorners=false
      bSeparateFillAlpha=false
      outlineMaterial=null
      outlineSize=1
      outlineColor=<rgba 0 0 0 0.4>
     }
    }
    shadowOffset=<0 0>
   }
   slot {class=HorizontalBoxSlot, properties {verticalAlignment=VAlign_Center, padding=<ltrb 0 0 16 0>}}
  }
  TutorialHighlight {widgetClassPath=&TutorialHighlight}
 ]
}
